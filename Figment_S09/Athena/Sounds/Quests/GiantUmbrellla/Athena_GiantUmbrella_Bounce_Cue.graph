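# Sound cue for bouncing on the giant umbrella quest prop, layering umbrella and giant beach ball bounce sounds.
# A mixer blends two modulated random branches, each picking one of three waves, with the beach ball layer mixed in quieter.
# It uses trap-launcher attenuation and the traps sound class, with a concurrency override that caps simultaneous plays and stops the oldest.
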SoundCue /Figment_S09/Athena/Sounds/Quests/GiantUmbrellla/Athena_GiantUmbrella_Bounce_Cue
root=Mixer_0
nodes [
 {id=Mixer_0, children [Modulator_0,Modulator_1], properties {inputVolume [1,0.4]}}
 {id=Modulator_0, children [Random_0]}
 {id=Modulator_1, children [Random_1]}
 {id=Random_0, children [WavePlayer_1,WavePlayer_2,WavePlayer_3], properties {weights [1,1,1]}}
 {id=Random_1, children [WavePlayer_4,WavePlayer_5,WavePlayer_6], properties {weights [1,1,1]}}
 {id=WavePlayer_1, properties {soundWaveAssetPtr=/Game/Athena/Sounds/Quests/GiantUmbrellla/Athena_GiantUmbrella_Bounce_01}}
 {id=WavePlayer_2, properties {soundWaveAssetPtr=/Game/Athena/Sounds/Quests/GiantUmbrellla/Athena_GiantUmbrella_Bounce_02}}
 {id=WavePlayer_3, properties {soundWaveAssetPtr=/Game/Athena/Sounds/Quests/GiantUmbrellla/Athena_GiantUmbrella_Bounce_03}}
 {id=WavePlayer_4, properties {soundWaveAssetPtr=/Figment_S09/Athena/Sounds/Quests/GiantBeachBall/Athena_GiantBeachBall_Bounce_01}}
 {id=WavePlayer_5, properties {soundWaveAssetPtr=/Figment_S09/Athena/Sounds/Quests/GiantBeachBall/Athena_GiantBeachBall_Bounce_02}}
 {id=WavePlayer_6, properties {soundWaveAssetPtr=/Figment_S09/Athena/Sounds/Quests/GiantBeachBall/Athena_GiantBeachBall_Bounce_03}}
]
properties {
 attenuationSettings=/Game/Sounds/Attenuation/Fort_Traps_Launch_Athena_Att
 maxDistance=7000
 soundClassObject=&Fort_Traps
 volumeMultiplier=1
 concurrencyOverrides {maxCount=3, resolutionRule=StopOldest, volumeScaleAttackTime=0.1}
}
